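# Physics asset for the F_MED_Booster female medium body, named for its cape and likely serving as body collision for cape simulation.
# Capsule bodies cover pelvis, upper spine, arms, hands, legs and feet, with only a tiny placeholder-sized capsule on spine_05; no cape bones have bodies.
# Constraints with uniform swing and twist limits chain the bodies back to the pelvis, except thigh_l, which has none, as calf_l links straight to the pelvis.
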PhysicsAsset /BRCosmetics/Characters/Player/Female/Medium/Bodies/F_MED_Booster/Meshes/F_MED_Booster_Physics_Cape
bodies {
 calf_l [{shapeType=Capsule, length=22.4868, radius=16.5184, center=<-16.4529 -2.2427 0>, rotation=<rot 90 0 0>}]
 calf_r [{shapeType=Capsule, length=22.4868, radius=16.5184, center=<16.4529 2.2427 0>, rotation=<rot 90 0 180>}]
 foot_l [{shapeType=Capsule, length=22.5766, radius=11.5396, center=<-3.9164 3.986 0>, rotation=<rot 0 14.9998 90>}]
 foot_r [{shapeType=Capsule, length=23.0766, radius=12.5396, center=<3.9164 -3.986 0>, rotation=<rot 0 14.9998 -90>}]
 hand_l [{shapeType=Capsule, length=9.5015, radius=6.7933, center=<6.2703 0 0>, rotation=<rot 90 0 0>}]
 hand_r [{shapeType=Capsule, length=9.5015, radius=6.7933, center=<-6.2703 0 0>, rotation=<rot 90 0 180>}]
 lowerarm_l [{shapeType=Capsule, length=17.4318, radius=5.9874, center=<8.9852 -0.6988 1.9278>, rotation=<rot 90 0 0>}]
 lowerarm_r [{shapeType=Capsule, length=16.4716, radius=7.0996, center=<-8.9852 0 -1.9278>, rotation=<rot 90 0 180>}]
 pelvis [{shapeType=Capsule, length=15, radius=15, shapeName=pelvis_capsule, center=<-2.4023 1.3076 -0.8401>}]
 spine_03 [{shapeType=Capsule, length=15.5621, radius=10.4944, center=<0 1.9166 0>}]
 spine_04 [{shapeType=Capsule, length=31.0169, radius=10.8628, center=<0 1.9166 0>}]
 spine_05 [{shapeType=Capsule, length=1, radius=1}]
 thigh_l [{shapeType=Capsule, length=17.2247, radius=12.8304, center=<-22.1102 0 1.2756>, rotation=<rot 90 0 180>}]
 thigh_r [{shapeType=Capsule, length=17.2247, radius=12.8304, center=<22.1102 0 -1.2756>, rotation=<rot 90 180 180>}]
 upperarm_l [{shapeType=Capsule, length=14.8344, radius=6.3144, center=<10.0455 0 0>, rotation=<rot 90 0 0>}]
 upperarm_r [{shapeType=Capsule, length=18.1746, radius=6.7467, center=<-10.0455 0 0>, rotation=<rot 90 0 180>}]
}
constraints [
 {bone1Name=spine_03, bone2Name=pelvis, swing1LimitDegrees=45, swing2LimitDegrees=45, twistLimitDegrees=45}
 {bone1Name=spine_04, bone2Name=spine_03, swing1LimitDegrees=45, swing2LimitDegrees=45, twistLimitDegrees=45}
 {bone1Name=spine_05, bone2Name=spine_04, swing1LimitDegrees=45, swing2LimitDegrees=45, twistLimitDegrees=45}
 {bone1Name=upperarm_l, bone2Name=spine_05, swing1LimitDegrees=45, swing2LimitDegrees=45, twistLimitDegrees=45}
 {bone1Name=lowerarm_l, bone2Name=upperarm_l, swing1LimitDegrees=45, swing2LimitDegrees=45, twistLimitDegrees=45}
 {bone1Name=hand_l, bone2Name=lowerarm_l, swing1LimitDegrees=45, swing2LimitDegrees=45, twistLimitDegrees=45}
 {bone1Name=upperarm_r, bone2Name=spine_05, swing1LimitDegrees=45, swing2LimitDegrees=45, twistLimitDegrees=45}
 {bone1Name=lowerarm_r, bone2Name=upperarm_r, swing1LimitDegrees=45, swing2LimitDegrees=45, twistLimitDegrees=45}
 {bone1Name=hand_r, bone2Name=lowerarm_r, swing1LimitDegrees=45, swing2LimitDegrees=45, twistLimitDegrees=45}
 {bone1Name=calf_l, bone2Name=pelvis, swing1LimitDegrees=45, swing2LimitDegrees=45, twistLimitDegrees=45}
 {bone1Name=foot_l, bone2Name=calf_l, swing1LimitDegrees=45, swing2LimitDegrees=45, twistLimitDegrees=45}
 {bone1Name=thigh_r, bone2Name=pelvis, swing1LimitDegrees=45, swing2LimitDegrees=45, twistLimitDegrees=45}
 {bone1Name=calf_r, bone2Name=thigh_r, swing1LimitDegrees=45, swing2LimitDegrees=45, twistLimitDegrees=45}
 {bone1Name=foot_r, bone2Name=calf_r, swing1LimitDegrees=45, swing2LimitDegrees=45, twistLimitDegrees=45}
]
