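# Post-process style AnimBlueprint for the F_MED_Booster cosmetic body, layering secondary motion onto the pose copied from its attached parent mesh.
# A Control Rig, fed by glide, slide, DBNO and Bendy-emote state from the EventGraph, precedes weighted chest and buckle constraints into a cached pose.
# Front end uses world-space AnimDynamics on both loop straps; otherwise bone-relative AnimDynamics feeds a RigidBody node using the cape physics asset.
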AnimBlueprint /BRCosmetics/Characters/Player/Female/Medium/Bodies/F_MED_Booster/Meshes/F_MED_Booster_AnimBP
target_skeleton=&/F_MED_Booster_Skeleton
graphs [
 AnimGraph {nodes [
   n1 = Anim Dynamics - Bone: base_loop_strap_mid_l!(ComponentPose=Local To Component(LocalPose=Use cached pose 'Common Input Cache'()), bAlphaBoolEnabled:bool=True, ExternalForce="0,0,0") -> (Pose) {
    alphaInputType=Bool
    angularDampingOverride=0.99
    angularSpringConstant=300
    bAngularSpring=true
    bOverrideAngularDamping=true
    bUseGravityOverride=true
    simulationSpace=BoneRelative
    alphaBoolBlend {blendOption=Linear}
    boundBone {boneName=base_loop_strap_mid_l}
    componentAppliedLinearAccClamp=<3000 3000 3000>
    componentLinearAccScale=<0.7 0.7 0.7>
    componentLinearVelScale=<0.7 0.7 0.7>
    gravityOverride {z=-980}
    physicsBodyDefinitions [{
      boundBone {boneName=base_loop_strap_mid_l}
      boxExtents=<10 10 10>
      constraintSetup {coneAngle=1, angularLimitsMax=<1 10 70>, angularLimitsMin=<-1 -10 5>, angularTarget=<1 0 0>}
      localJointOffset=<1 0 0>
     }]
    relativeSpaceBone {boneName=root}
   }
   n2 = Anim Dynamics - Bone: base_loop_strap_mid_r!(ComponentPose=@n1, bAlphaBoolEnabled:bool=True, ExternalForce="0,0,0") -> (Pose) {
    alphaInputType=Bool
    angularDampingOverride=0.99
    angularSpringConstant=300
    bAngularSpring=true
    bOverrideAngularDamping=true
    bUseGravityOverride=true
    simulationSpace=BoneRelative
    alphaBoolBlend {blendOption=Linear}
    boundBone {boneName=base_loop_strap_mid_r}
    componentAppliedLinearAccClamp=<3000 3000 3000>
    componentLinearAccScale=<0.7 0.7 0.7>
    componentLinearVelScale=<0.7 0.7 0.7>
    gravityOverride {z=-980}
    physicsBodyDefinitions [{
      boundBone {boneName=base_loop_strap_mid_r}
      boxExtents=<10 10 10>
      constraintSetup {coneAngle=1, angularLimitsMax=<1 10 70>, angularLimitsMin=<-1 -10 5>, angularTarget=<1 0 0>}
      localJointOffset=<1 0 0>
     }]
    relativeSpaceBone {boneName=root}
   }
   n3 = RigidBody!(ComponentPose=@n2, bAlphaBoolEnabled:bool, ExternalForce) -> (Pose) {
    alphaInputType=Bool
    overridePhysicsAsset=&/F_MED_Booster_Physics_Cape
    alphaBoolBlend {blendOption=Linear}
   }
   n4 = Anim Dynamics - Bone: base_loop_strap_mid_l(
    ComponentPose=Local To Component(LocalPose=Use cached pose 'Common Input Cache'())
    bAlphaBoolEnabled=Get bIsInFrontEnd()
    ExternalForce="0,0,0"
   ) {
    alphaInputType=Bool
    angularDampingOverride=0.99
    angularSpringConstant=300
    bAngularSpring=true
    bOverrideAngularDamping=true
    bUseGravityOverride=true
    simulationSpace=World
    alphaBoolBlend {blendOption=Linear}
    boundBone {boneName=base_loop_strap_mid_l}
    componentAppliedLinearAccClamp=<3000 3000 3000>
    componentLinearAccScale=<0.5 0.5 0.5>
    componentLinearVelScale=<0.5 0.5 0.5>
    gravityOverride {z=-980}
    physicsBodyDefinitions [{
      boundBone {boneName=base_loop_strap_mid_l}
      boxExtents=<12 12 12>
      constraintSetup {coneAngle=1, angularLimitsMax=<1 10 70>, angularLimitsMin=<-1 -10 5>, angularTarget=<1 0 0>}
      localJointOffset=<0.6 0 0>
     }]
    relativeSpaceBone {boneName=root}
   }
   n5 = Anim Dynamics - Bone: base_loop_strap_mid_r(ComponentPose=@n4, bAlphaBoolEnabled=Get bIsInFrontEnd(), ExternalForce="0,0,0") {
    alphaInputType=Bool
    angularDampingOverride=0.99
    angularSpringConstant=300
    bAngularSpring=true
    bOverrideAngularDamping=true
    bUseGravityOverride=true
    simulationSpace=World
    alphaBoolBlend {blendOption=Linear}
    boundBone {boneName=base_loop_strap_mid_r}
    componentAppliedLinearAccClamp=<3000 3000 3000>
    componentLinearAccScale=<0.5 0.5 0.5>
    componentLinearVelScale=<0.5 0.5 0.5>
    gravityOverride {z=-980}
    physicsBodyDefinitions [{
      boundBone {boneName=base_loop_strap_mid_r}
      boxExtents=<12 12 12>
      constraintSetup {coneAngle=1, angularLimitsMax=<1 10 70>, angularLimitsMin=<-1 -10 5>, angularTarget=<1 0 0>}
      localJointOffset=<0.6 0 0>
     }]
    relativeSpaceBone {boneName=root}
   }
   Output Pose(Result=Blend Poses by bool(BlendPose_0=Component To Local(ComponentPose=@n5), BlendPose_1=Component To Local(ComponentPose=@n3), bActiveValue=Get bIsInFrontEnd()) {blendType=Linear}) {layerGroup=None}
   n6 = Control Rig!(
    Source=Copy Pose From Mesh() {bUseAttachedParent=true}
    Alpha
    HangTwo arms:double=Get HangTwo arms() -> (HangTwo arms:double)
    SlideAlpha:double=Get SlideAlpha() -> (SlideAlpha:double)
    isDBNO:bool=Get bIsDBNO()
    IsBendy:bool=Get ISBandy() -> (ISBandy:bool)
   ) -> (Pose) {
    alphaBoolBlend {blendOption=Linear}
    controlRigAssetReference {blueprintRigClass=&/F_MED_Booster_CtrlRig.F_MED_Booster_CtrlRig_C}
   }
   n7 = "Constraint - MSTR_brest_l (PPPPPPPPP)"!(
    ComponentPose=Local To Component(LocalPose=@n6)
    Alpha=1
    ConstraintWeights_0=0.2788
    ConstraintWeights_1:float=0.2816
    ConstraintWeights_2:float=0.1252
    ConstraintWeights_3:float=0.0173
    ConstraintWeights_4:float=0.0038
    ConstraintWeights_5:float=0.0849
    ConstraintWeights_6:float=0.0655
    ConstraintWeights_7:float=0.0607
    ConstraintWeights_8:float=0.0822
   ) -> (Pose) {
    alphaBoolBlend {blendOption=Linear}
    boneToModify {boneName=MSTR_brest_l}
    constraintSetup [
     {targetBone {boneName=spine_05}}
     {targetBone {boneName=clavicle_l}}
     {targetBone {boneName=spine_04}}
     {targetBone {boneName=upperarm_twist_01_l}}
     {targetBone {boneName=deform_upperarm_bk_l}}
     {targetBone {boneName=deform_upperarm_fr_l}}
     {targetBone {boneName=deform_upperarm_up_l}}
     {targetBone {boneName=deform_pec_l}}
     {targetBone {boneName=Collar_dyn_l}}
    ]
    constraintWeights [1,1,1,1,1,1,1,1,1]
   }
   n8 = "Constraint - MSTR_brest_r (PPPPPPP)"!(
    ComponentPose=@n7
    Alpha=1
    ConstraintWeights_0=0.1931
    ConstraintWeights_1:float=0.5411
    ConstraintWeights_2:float=0.0809
    ConstraintWeights_3:float=0
    ConstraintWeights_4:float=0.1053
    ConstraintWeights_5:float=0.0391
    ConstraintWeights_6:float=0.0404
   ) -> (Pose) {
    alphaBoolBlend {blendOption=Linear}
    boneToModify {boneName=MSTR_brest_r}
    constraintSetup [
     {targetBone {boneName=spine_05}}
     {targetBone {boneName=clavicle_r}}
     {targetBone {boneName=spine_04}}
     {targetBone {boneName=deform_upperarm_bk_r}}
     {targetBone {boneName=deform_upperarm_fr_r}}
     {targetBone {boneName=deform_upperarm_up_r}}
     {targetBone {boneName=deform_pec_r}}
    ]
    constraintWeights [1,1,1,1,1,1,1]
   }
   n9 = "Constraint - Base_buclket_l (PPPPPPPPPPPPPPPP)"!(
    ComponentPose=@n8
    Alpha=1
    ConstraintWeights_0=0
    ConstraintWeights_1:float=0
    ConstraintWeights_2:float=0.1407
    ConstraintWeights_3:float=0.0007
    ConstraintWeights_4:float=0.601
    ConstraintWeights_5:float=0.0028
    ConstraintWeights_6:float=0.0127
    ConstraintWeights_7:float=0.0012
    ConstraintWeights_8:float=0
    ConstraintWeights_9:float=0.0223
    ConstraintWeights_10:float=0.0005
    ConstraintWeights_11:float=0.0363
    ConstraintWeights_12:float=0.0006
    ConstraintWeights_13:float=0.0001
    ConstraintWeights_14:float=0.007
    ConstraintWeights_15:float=0.174
   ) -> (Pose) {
    alphaBoolBlend {blendOption=Linear}
    boneToModify {boneName=Base_buclket_l}
    constraintSetup [
     {targetBone {boneName=head}}
     {targetBone {boneName=neck_01}}
     {targetBone {boneName=spine_05}}
     {targetBone {boneName=clavicle_l}}
     {targetBone {boneName=clavicle_r}}
     {targetBone {boneName=spine_04}}
     {targetBone {boneName=upperarm_twist_01_r}}
     {targetBone {boneName=spine_03}}
     {targetBone {boneName=deform_groin_r}}
     {targetBone {boneName=deform_upperarm_bk_r}}
     {targetBone {boneName=deform_upperarm_fr_r}}
     {targetBone {boneName=deform_upperarm_up_r}}
     {targetBone {boneName=Sholderpad_dyn_r}}
     {targetBone {boneName=Collar_dyn_r}}
     {targetBone {boneName=Collar_dyn_bk}}
     {targetBone {boneName=Corr_buclket_01}}
    ]
    constraintWeights [1,1,1,1,1,1,1,1,1,1,1,1,1,1,1,1]
   }
   Save cached pose 'Common Input Cache'(Pose=Component To Local(ComponentPose=@n9)) {cacheName=Common Input Cache}
  ]}
 EventGraph {nodes [
   Event BlueprintUpdateAnimation()
   > n1 = Sequence!(execute) -> (then_0, then_1, then_2:Exec)
   Set is Wearing Cape(execute=@n1.then_1, is Wearing Cape:bool=CheckForBackbling().bIsBackblingACape) -> (Output_Get:bool)
   n2 = GetOwnerSkeletalMeshComponent()
   n3 = Is Valid(exec=@n1.then_2, InputObject=@n2)
   Set ISBandy(execute=@n3.Is Not Valid, ISBandy:bool=false) -> (Output_Get:bool)
   Set ISBandy(
    execute=@n3.Is Valid
    ISBandy=Montage_IsPlaying(self=GetAnimInstance(self=@n2), Montage=/BRCosmetics/Animation/Game/MainPlayer/Emotes/Bendy/Emote_Bendy_CMF_M)
   )
   n4 = Gravity Override State Function DEPRECATED(
    execute=@n1.then_0
    CharacterPartAnimInstance=Self-Reference()
    GameTimeInSeconds=GetGameTimeInSeconds()
    GravityOverrideParams=Get Gravity Override Params()
   )
   > Set Gravity(Gravity:Vector=@n4) -> (Output_Get:Vector)
   Event OnBeginKneeSliding() {is=title K2Node_Event}
   > Set SlideAlpha(SlideAlpha:double=1) -> (Output_Get:double)
   Event OnEndKneeSliding() {is=title K2Node_Event}
   > Set SlideAlpha(SlideAlpha=0)
   Event OnParachuteClosed()
   > Set HangTwo arms(HangTwo arms:double=0) -> (Output_Get:double)
   Event OnEndBaseballSliding() {is=title K2Node_Event}
   > Set SlideAlpha(SlideAlpha=0)
   Event OnParachuteOpened()
   > Switch on EGliderType(Selection=Get PlayerGliderType() -> (PlayerGliderType=HangGlider))
   > Set HangTwo arms(HangTwo arms=1)
   n5 = Branch(Condition=Get bIsCrouching())
   > Set suspenders_component_linear(suspenders_component_linear:Vector="0, 0, 0") -> (Output_Get:Vector)
   Set suspenders_component_linear(execute=@n5.else, suspenders_component_linear="0, 0, 0")
  ]}
]
variables [
 Gravity {type=Vector}
 Gravity Override Params {type=GravityOverrideParamsStruct}
 HangTwo arms {type=number}
 is Wearing Cape {type=boolean}
 ISBandy {type=boolean}
 SlideAlpha {type=number}
 suspenders_component_linear {type=Vector}
 Wear_cape {type=boolean}
]
