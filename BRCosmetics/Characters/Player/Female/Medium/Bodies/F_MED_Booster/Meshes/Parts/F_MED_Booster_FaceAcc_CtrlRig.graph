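# Control Rig for the F_MED_Booster face accessory that rotates the MSTR_ponytail bone to follow head pose, backpack and diving state.
# It branches on "Is backpack", with both paths using a spherical pose reader on the head to weight a ponytail rotation, much larger with a backpack.
# After the branch, DivingDown weights another ponytail rotation.
# Kalman filters smooth all three weights, and debug drawing remains enabled on the no-backpack reader.
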ControlRigBlueprint /BRCosmetics/Characters/Player/Female/Medium/Bodies/F_MED_Booster/Meshes/Parts/F_MED_Booster_FaceAcc_CtrlRig
graphs [
 RigVMFunctionLibrary {links [], nodes []}
 RigVMModel {nodes [
   RigUnit_BeginExecution = &RigUnit_BeginExecution()
   RigVMFunction_ControlFlowBranch = &RigVMFunction_ControlFlowBranch(ExecuteContext=@RigUnit_BeginExecution.ExecutePin, Condition=RigVMVariableNode() {variable=Is backpack}.Value)
   &RigUnit_ModifyTransforms(
    ItemToModify=((Item=(Type=Bone,Name="MSTR_ponytail"),Transform=(Rotation=(X=0,Y=0,Z=-0.342,W=0.9397),Translation=(X=0,Y=0,Z=0),Scale3D=(X=1,Y=1,Z=1))))
    WeightMinimum=0
    Weight=&RigVMFunction_KalmanFloat(BufferSize=20, Value=RigVMVariableNode() {variable=DivingDown}.Value) {template="KalmanFilter::Execute(in Value,in BufferSize,out Result)"}.Result
    ExecutePin=@RigVMFunction_ControlFlowBranch.Completed
   )
   n1 = &RigUnit_SphericalPoseReader(
    ActiveRegionScaleFactors=(PositiveWidth=1,NegativeWidth=1,PositiveHeight=1,NegativeHeight=1)
    ActiveRegionSize=0.04
    Debug=(bDrawDebug=True,bDraw2D=False,bDrawLocalAxes=False,DebugScale=25,DebugSegments=20,DebugThickness=0.25)
    DriverAxis=(X=1,Y=0,Z=0)
    DriverItem=(Type=Bone,Name="head")
    FalloffRegionScaleFactors=(PositiveWidth=1,NegativeWidth=0.4,PositiveHeight=1,NegativeHeight=1)
    FalloffSize=0.1
    FlipHeightScaling=False
    FlipWidthScaling=False
    OptionalParentItem=(Type=Bone,Name="None")
    RotationOffset=(X=90,Y=0,Z=-103)
    ExecutePin=@RigVMFunction_ControlFlowBranch.False
   )
   > &RigUnit_ModifyTransforms(
    ItemToModify=((Item=(Type=Bone,Name="MSTR_ponytail"),Transform=(Rotation=(X=0,Y=0,Z=-0.0436,W=0.999),Translation=(X=0,Y=0,Z=0),Scale3D=(X=1,Y=1,Z=1))))
    WeightMinimum=0
    Weight=&RigVMFunction_KalmanFloat(BufferSize=16, Value=@n1.OutputParam) {template="KalmanFilter::Execute(in Value,in BufferSize,out Result)"}.Result
   )
   n2 = &RigUnit_SphericalPoseReader(
    ActiveRegionScaleFactors=(PositiveWidth=1,NegativeWidth=1,PositiveHeight=1,NegativeHeight=1)
    ActiveRegionSize=0.04
    Debug=(bDrawDebug=false,bDraw2D=False,bDrawLocalAxes=False,DebugScale=25,DebugSegments=20,DebugThickness=0.25)
    DriverAxis=(X=1,Y=0,Z=0)
    DriverItem=(Type=Bone,Name="head")
    FalloffRegionScaleFactors=(PositiveWidth=1,NegativeWidth=0.4,PositiveHeight=1,NegativeHeight=1)
    FalloffSize=0.1
    FlipHeightScaling=False
    FlipWidthScaling=False
    OptionalParentItem=(Type=Bone,Name="None")
    RotationOffset=(X=90,Y=0,Z=-103)
    ExecutePin=@RigVMFunction_ControlFlowBranch.True
   )
   KalmanFilter = &RigVMFunction_KalmanFloat(BufferSize=16, Value=@n2.OutputParam) {template="KalmanFilter::Execute(in Value,in BufferSize,out Result)"}
   > &RigUnit_ModifyTransforms(
    ItemToModify=((Item=(Type=Bone,Name="MSTR_ponytail"),Transform=(Rotation=(X=0,Y=0,Z=-0.3987,W=0.9171),Translation=(X=0,Y=0,Z=0),Scale3D=(X=1,Y=1,Z=1))))
    WeightMinimum=0
    Weight=@KalmanFilter.Result
   )
  ]}
]
hierarchy [
 root {loc=<0 0 0>, rot=<0 0 0 1>}
 pelvis {parent=root, loc=<0 -0.5301 98.0571>, rot=<0.0127 0.707 -0.0127 -0.707>}
 spine_01 {parent=pelvis, loc=<6.1208 -0.9328 0>, rot=<0 0 0.034 -0.9994>}
 spine_02 {parent=spine_01, loc=<6.5997 -0.2095 0>, rot=<0 0 0.0238 -0.9997>}
 spine_03 {parent=spine_02, loc=<6.4914 -0.1237 0>, rot=<0 0 0.0149 -0.9999>}
 spine_04 {parent=spine_03, loc=<6.9598 0.0794 0>, rot=<0 0 -0.0132 -0.9999>}
 spine_05 {parent=spine_04, loc=<6.9987 0 0>, rot=<0 0 -0.0015 -1>}
 neck_01 {parent=spine_05, loc=<11.3138 0.1529 0>, rot=<0 0 -0.2003 -0.9797>}
 neck_02 {parent=neck_01, loc=<4.5627 0 0>, rot=<0 0 0.0504 -0.9987>}
 head {parent=neck_02, loc=<5.3914 0 0>, rot=<0 0 0.115 -0.9934>}
 Base_hair_fr_l {parent=head, loc=<6.2546 8.6664 -3.2869>, rot=<-0.0222 -0.8326 0.3353 -0.4403>}
 dyn_hair_fr_l_01 {parent=Base_hair_fr_l, loc=<0 0 0>, rot=<0 0 0 1>}
 Base_hair_side_l {parent=head, loc=<6.4868 4.5448 -6.5566>, rot=<-0.0014 -0.812 0.5653 -0.145>}
 dyn_hair_side_l_01 {parent=Base_hair_side_l, loc=<0 0 0>, rot=<0 0 0 1>}
 dyn_hair_side_l_02 {parent=dyn_hair_side_l_01, loc=<5.0057 -0.7664 0.4947>, rot=<-0.0968 -0.0905 0.0145 -0.9911>}
 dyn_hair_side_l_03 {parent=dyn_hair_side_l_02, loc=<3.9564 -0.2408 -0.55>, rot=<0 0 0.0471 -0.9989>}
 Base_hair_side_r {parent=head, loc=<5.612 3.2541 6.4781>, rot=<0.0121 -0.7315 -0.6812 -0.0254>}
 dyn_hair_side_r_01 {parent=Base_hair_side_r, loc=<0 0 0>, rot=<0 0 0 1>}
 dyn_hair_side_r_02 {parent=dyn_hair_side_r_01, loc=<3.724 0.6655 -0.3184>, rot=<0 0 0 1>}
 dyn_hair_side_r_03 {parent=dyn_hair_side_r_02, loc=<3.4783 0.0563 0.2532>, rot=<0 0 0 1>}
 MSTR_ponytail {parent=head, loc=<7.1111 -9.3465 0.069>, rot=<0.71 -0.7042 0 0>}
 Base_ponytail {parent=MSTR_ponytail, loc=<0 0 0>, rot=<0 0 0 1>}
 dyn_ponytail_A {parent=Base_ponytail, loc=<1.7278 -0.0573 -3.1281>, rot=<0 -0.2882 0 -0.9576>}
 dyn_ponytail_B {parent=Base_ponytail, loc=<2.2284 0 2.8926>, rot=<0 0 0 1>}
 dyn_ponytail_C {parent=Base_ponytail, loc=<0.6669 2.5707 -1.6586>, rot=<0 0 0 1>}
 dyn_ponytail_D {parent=Base_ponytail, loc=<1.2202 3.3948 2.3885>, rot=<0 0 0 1>}
 clavicle_l {parent=spine_05, loc=<5.4831 -0.1053 -3.2946>, rot=<0.0465 -0.7289 -0.009 -0.6829>}
 clavicle_r {parent=spine_05, loc=<5.4831 -0.1053 3.2946>, rot=<0.7289 0.0465 -0.6829 0.009>}
 upperarm_r {parent=clavicle_r, loc=<-9.8658 1.2136 0.4105>, rot=<0.0164 -0.3245 -0.0269 -0.9454>}
 upperarm_twist_01_r {parent=upperarm_r, loc=<-7.2135 0.1134 0>, rot=<0 0 0 -1>}
 Collar_base_l {parent=spine_05, loc=<8.7551 3.8456 -7.8828>, rot=<0.3541 -0.0573 0.9219 -0.1463>}
 Collar_dyn_l {parent=Collar_base_l, loc=<0 0 0>, rot=<0 0 0 1>}
 Collar_base_bk {parent=spine_05, loc=<10.2397 -5.899 0>, rot=<0 0 0.966 0.2584>}
 Collar_dyn_bk {parent=Collar_base_bk, loc=<0 0 0>, rot=<0 0 0 1>}
 Collar_base_r {parent=spine_05, loc=<8.755 3.8456 7.8828>, rot=<-0.3541 0.0573 0.9219 -0.1463>}
 Collar_dyn_r {parent=Collar_base_r, loc=<0 0 0>, rot=<0 0 0 1>}
]
variables [
 DivingDown {type=double, default=0}
 Is backpack {type=bool, default=False}
]
