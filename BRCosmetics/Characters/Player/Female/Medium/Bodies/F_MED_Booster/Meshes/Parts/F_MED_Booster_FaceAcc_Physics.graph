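# Physics asset for the female medium Booster character's face accessory part, giving its hair, ponytail and collar bones physics bodies.
# Single-capsule bodies form constraint chains from the head: three-segment left and right side locks, a short front-left lock, and four ponytail strands branching in parallel from one base.
# Joints use tight angular limits, except the free front-lock root and twist-locked ponytail base.
# The three collar capsules have no constraints.
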PhysicsAsset /BRCosmetics/Characters/Player/Female/Medium/Bodies/F_MED_Booster/Meshes/Parts/F_MED_Booster_FaceAcc_Physics
bodies {
 Base_hair_fr_l [{shapeType=Capsule, length=0.505, radius=0.505, shapeName=Base_hair_fr_l_capsule, rotation=<rot 0 -90 0>}]
 Base_hair_side_l [{shapeType=Capsule, length=0.505, radius=0.505, shapeName=Base_hair_side_l_capsule, rotation=<rot 0 -90 0>}]
 Base_hair_side_r [{shapeType=Capsule, length=0.505, radius=0.505, shapeName=Base_hair_side_r_capsule, rotation=<rot 0 -90 0>}]
 Base_ponytail [{shapeType=Capsule, length=0.6906, radius=7.8447, center=<7.6061 2.741 0>}]
 Collar_base_l [{
   shapeType=Capsule
   length=8.9534
   radius=3.2493
   shapeName=Collar_base_l_capsule
   center=<-1.0915 -0.4365 -0.7072>
   rotation=<rot 44.9572 -176.8617 101.2815>
  }]
 Collar_base_r [{
   shapeType=Capsule
   length=14.1475
   radius=3.2444
   shapeName=Collar_base_r_capsule
   center=<-1.5042 -0.2307 0.2539>
   rotation=<rot 44.8817 -5.2082 94.0211>
  }]
 Collar_dyn_bk [{shapeType=Capsule, length=12.6865, radius=5.8439, shapeName=Collar_dyn_bk_capsule, center=<1.0933 -0.5639 0>, rotation=<rot 0 -151.6193 0>}]
 dyn_hair_fr_l_01 [{
   shapeType=Capsule
   length=6.5521
   radius=1.1761
   shapeName=dyn_hair_fr_l_01_capsule
   center=<1.1643 0.3734 0.4289>
   rotation=<rot -3.5615 96.9953 -78.4182>
  }]
 dyn_hair_side_l_01 [{shapeType=Capsule, length=2.1617, radius=1.1925, shapeName=dyn_hair_side_l_01_capsule, center=<0 0.1886 0>, rotation=<rot 0 -90 -6.4772>}]
 dyn_hair_side_l_02 [{
   shapeType=Capsule
   length=1.9944
   radius=1.7354
   shapeName=dyn_hair_side_l_02_capsule
   center=<0 0.6781 -0.0964>
   rotation=<rot 8.0932 -90 0.6854>
  }]
 dyn_hair_side_l_03 [{
   shapeType=Capsule
   length=5.0944
   radius=1.3856
   shapeName=dyn_hair_side_l_03_capsule
   center=<1.34 -0.2398 0.0937>
   rotation=<rot -0.3431 84.8417 -86.0757>
  }]
 dyn_hair_side_r_01 [{
   shapeType=Capsule
   length=3.9583
   radius=1.2508
   shapeName=dyn_hair_side_r_01_capsule
   center=<-0.1743 0.5682 0>
   rotation=<rot 0 -72.9443 -73.715>
  }]
 dyn_hair_side_r_02 [{shapeType=Capsule, length=2.0928, radius=1.161, shapeName=dyn_hair_side_r_02_capsule, rotation=<rot 0 -90 85.668>}]
 dyn_hair_side_r_03 [{
   shapeType=Capsule
   length=5.5345
   radius=1.5718
   shapeName=dyn_hair_side_r_03_capsule
   center=<2.651 0.6401 -0.517>
   rotation=<rot 0 -85.1265 -79.144>
  }]
 dyn_ponytail_A [{
   shapeType=Capsule
   length=5.4936
   radius=1.5877
   shapeName=dyn_ponytail_A_capsule
   center=<5.9848 2.6144 3.553>
   rotation=<rot 0 141.4208 -77.359>
  }]
 dyn_ponytail_B [{
   shapeType=Capsule
   length=5.5776
   radius=1.6419
   shapeName=dyn_ponytail_B_capsule
   center=<2.152 1.4915 -0.1571>
   rotation=<rot 0 -55.2753 -86.566>
  }]
 dyn_ponytail_C [{
   shapeType=Capsule
   length=4.9014
   radius=3.3664
   shapeName=dyn_ponytail_C_capsule
   center=<2.8771 1.9281 -1.9129>
   rotation=<rot 0 -51.5062 -65.8462>
  }]
 dyn_ponytail_D [{
   shapeType=Capsule
   length=8.5908
   radius=3.5333
   shapeName=dyn_ponytail_D_capsule
   center=<5.4944 2.3763 -0.6976>
   rotation=<rot 0 -56.3468 -66.9386>
  }]
 head [{shapeType=Capsule, length=6.8528, radius=5.9242, center=<2.5318 1.6421 0.0033>, rotation=<rot 0 -23.5296 -89.6808>}]
}
constraints [
 {bone1Name=Base_ponytail, bone2Name=head, swing1LimitDegrees=40, swing2LimitDegrees=25, twistLimitDegrees=3, twistMotion=Locked}
 {bone1Name=Base_hair_side_l, bone2Name=head, swing1LimitDegrees=7, swing2LimitDegrees=20, twistLimitDegrees=5}
 {bone1Name=dyn_hair_side_l_01, bone2Name=Base_hair_side_l, swing1LimitDegrees=7, swing2LimitDegrees=20, twistLimitDegrees=5}
 {bone1Name=dyn_hair_side_l_02, bone2Name=dyn_hair_side_l_01, swing1LimitDegrees=7, swing2LimitDegrees=20, twistLimitDegrees=5}
 {bone1Name=dyn_hair_side_l_03, bone2Name=dyn_hair_side_l_02, swing1LimitDegrees=7, swing2LimitDegrees=20, twistLimitDegrees=5}
 {bone1Name=Base_hair_side_r, bone2Name=head, swing1LimitDegrees=7, swing2LimitDegrees=20, twistLimitDegrees=5}
 {bone1Name=dyn_hair_side_r_01, bone2Name=Base_hair_side_r, swing1LimitDegrees=7, swing2LimitDegrees=20, twistLimitDegrees=5}
 {bone1Name=dyn_hair_side_r_02, bone2Name=dyn_hair_side_r_01, swing1LimitDegrees=7, swing2LimitDegrees=20, twistLimitDegrees=5}
 {bone1Name=dyn_hair_side_r_03, bone2Name=dyn_hair_side_r_02, swing1LimitDegrees=7, swing2LimitDegrees=20, twistLimitDegrees=5}
 {bone1Name=dyn_ponytail_A, bone2Name=Base_ponytail, swing1LimitDegrees=15, swing2LimitDegrees=15, twistLimitDegrees=5}
 {bone1Name=dyn_ponytail_D, bone2Name=Base_ponytail, swing1LimitDegrees=15, swing2LimitDegrees=15, twistLimitDegrees=5}
 {bone1Name=dyn_ponytail_B, bone2Name=Base_ponytail, swing1LimitDegrees=15, swing2LimitDegrees=15, twistLimitDegrees=5}
 {bone1Name=dyn_ponytail_C, bone2Name=Base_ponytail, swing1LimitDegrees=7, swing2LimitDegrees=7, twistLimitDegrees=5}
 {
  bone1Name=Base_hair_fr_l
  bone2Name=head
  swing1LimitDegrees=45
  swing1Motion=Free
  swing2LimitDegrees=45
  swing2Motion=Free
  twistLimitDegrees=45
  twistMotion=Free
 }
 {bone1Name=dyn_hair_fr_l_01, bone2Name=Base_hair_fr_l, swing1LimitDegrees=5, swing2LimitDegrees=15, twistLimitDegrees=2}
]
