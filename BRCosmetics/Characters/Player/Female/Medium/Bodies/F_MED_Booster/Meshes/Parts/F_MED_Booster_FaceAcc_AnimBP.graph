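# Anim Blueprint for the F_MED_Booster face accessory part, adding secondary physics motion.
# Copies the attached parent's pose, layers Body's clavicle branches over it, then runs a Control Rig fed backpack and dive state.
# The cached result feeds two RigidBody sims, selected by bIsInFrontEnd: world-space with gravity for front end, head-space zero-gravity otherwise.
# EventGraph updates backpack, DBNO and dive values, using a deprecated gravity-state function.
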AnimBlueprint /BRCosmetics/Characters/Player/Female/Medium/Bodies/F_MED_Booster/Meshes/Parts/F_MED_Booster_FaceAcc_AnimBP
target_skeleton=&/F_MED_Booster_FaceAcc_Skeleton
graphs [
 AnimGraph {nodes [
   n1 = RigidBody!(
    ComponentPose=Local To Component(LocalPose=Use cached pose 'Comon input Cache'())
    Alpha
    OverrideWorldGravity:Vector="0,0,-980"
    ExternalForce
   ) -> (Pose) {
    bClampLinearTranslationLimitToRefPose=true
    bOverrideWorldGravity=true
    overridePhysicsAsset=&/F_MED_Booster_FaceAcc_Physics
    simulationSpace=WorldSpace
    alphaBoolBlend {blendOption=Linear}
    baseBoneRef {boneName=root}
    simSpaceSettings {maxAngularAcceleration=250, maxAngularVelocity=5, maxLinearAcceleration=250, maxLinearVelocity=700, worldAlpha=1}
   }
   n2 = RigidBody(ComponentPose=Local To Component(LocalPose=Use cached pose 'Comon input Cache'()), OverrideWorldGravity="0,0,0") {
    bClampLinearTranslationLimitToRefPose=true
    bOverrideWorldGravity=true
    overridePhysicsAsset=&F_MED_Booster_FaceAcc_Physics
    simulationSpace=BaseBoneSpace
    alphaBoolBlend {blendOption=Linear}
    baseBoneRef {boneName=head}
    componentAppliedLinearAccClamp=<1500 1500 1500>
    componentLinearAccScale=<0.1 0.1 0.36>
    componentLinearVelScale=<0.1 0.1 0.36>
    simSpaceSettings {maxAngularAcceleration=250, maxAngularVelocity=3, maxLinearAcceleration=250, maxLinearVelocity=250, velocityScaleZ=0, worldAlpha=1}
   }
   Output Pose(Result=Blend Poses by bool(BlendPose_0=Component To Local(ComponentPose=@n1), BlendPose_1=Component To Local(ComponentPose=@n2), bActiveValue=Get bIsInFrontEnd()) {blendType=Linear}) {layerGroup=None}
   n3 = Get Body() -> (Body:Skeletal Mesh Component Object Reference)
   n4 = Control Rig!(
    Source=Layered blend per bone(
     BasePose=Copy Pose From Mesh() {bCopyCurves=true, bUseAttachedParent=true}
     BlendPoses_0=Copy Pose From Mesh(SourceMeshComponent=@n3) {bUseAttachedParent=true}
    ) {layerSetup [{branchFilters [{blendDepth=2, boneName=clavicle_r}, {blendDepth=2, boneName=clavicle_l}]}]}
    Alpha
    Is backpack:bool=Get is Wearing CapeBackpack() -> (is Wearing CapeBackpack:bool)
    DivingDown:double=Get DivingDown Alpha() -> (DivingDown Alpha:double)
   ) -> (Pose) {
    alphaBoolBlend {blendOption=Linear}
    controlRigAssetReference {blueprintRigClass=&/F_MED_Booster_FaceAcc_CtrlRig.F_MED_Booster_FaceAcc_CtrlRig_C}
   }
   Save cached pose 'Comon input Cache'(Pose=@n4) {cacheName=Comon input Cache}
  ]}
 EventGraph {nodes [
   Event BlueprintUpdateAnimation()
   > n1 = Sequence!(execute) -> (then_0, then_1, then_2:Exec)
   > n2 = Gravity Override State Function DEPRECATED(CharacterPartAnimInstance=Self-Reference(), GameTimeInSeconds=GetGameTimeInSeconds(), GravityOverrideParams=Get Gravity Override Params())
   > Set Gravity Override(Gravity Override=@n2)
   > n3 = Branch(Condition=Get bIsDBNO())
   > Set DBNO alpha(DBNO alpha:double=0.8) -> (Output_Get:double)
   Set DBNO alpha(execute=@n3.else, DBNO alpha=1)
   n4 = Branch(execute=@n1.then_2, Condition="Equal Exactly (String)"(A=@n2.CurrentStateOut, B=SkyDiving_DivingDown))
   > Set DivingDown Alpha(DivingDown Alpha:double=1) -> (Output_Get:double)
   Set DivingDown Alpha(execute=@n4.else, DivingDown Alpha=0)
   Set is Wearing CapeBackpack(execute=@n1.then_1, is Wearing CapeBackpack:bool=CheckForBackbling()) -> (Output_Get:bool)
   Event BlueprintInitializeAnimation()
   > Set body(body:Skeletal Mesh Component Object Reference=GetOwnerSkeletalMeshComponent()) -> (Output_Get:Skeletal Mesh Component Object Reference)
   TryGetPawnOwner()
  ]}
]
variables [
 body {type=SkeletalMeshComponent}
 DBNO alpha {type=number}
 DivingDown Alpha {type=number}
 Gravity Override {type=Vector}
 Gravity Override Params {type=GravityOverrideParamsStruct}
 is Wearing CapeBackpack {type=boolean}
]
